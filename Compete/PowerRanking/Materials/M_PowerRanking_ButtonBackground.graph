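# Translucent UI material for a Power Ranking button background, with animated blue bars and a glowing sweep inside a rounded-corner shape.
# Panned horizontal UVs form repeating bars that strengthen rightward, alongside a scaled blurred-noise signal.
# A time-looped reveal sweep with a configurable gap derives edge, glow and pattern masks, shared through named reroutes.
# A combined mask drives the navy-to-bright-blue blend under Color Dodge, and its power curve multiplies the rounded mask.
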Material /Compete/PowerRanking/Materials/M_PowerRanking_ButtonBackground
graph {
 properties {bAutomaticallySetUsageInEditor=false, blendMode=BLEND_Translucent, materialDomain=MD_UI, usedShadingModels=Default Lit}
 nodes [
  n1 = NamedRerouteUsage()
  n2 = NamedRerouteUsage()
  n3 = NamedRerouteUsage()
  n4 = NamedRerouteUsage()
  n5 = LinearInterpolate(A=LinearInterpolate(A=Multiply(A=@n2, B=@n3), Alpha=NamedRerouteUsage()), Alpha=Multiply(A=Multiply(A=@n4, B=@n4)) {constB=0.2})
  MaterialFunctionCall!(
   Blend=Min(A=Power(Base=@n1) {constExponent=4}, B=Constant() {r=0.8}) {constB=0.8}
   Base=LinearInterpolate(A=Constant3Vector() {constant=<rgba 0.006 0.0103 0.0742 1>}, B=Constant3Vector() {constant=<rgba 0.0976 0.5089 1 1>}, Alpha=@n5)
  ) {materialFunction=/Engine/Functions/Engine_MaterialFunctions03/Blends/Blend_ColorDodge}
  Multiply(
   A=Power(Base=@n5) {constExponent=0.2}
   B=MaterialFunctionCall(Corner Radius=Constant() {r=10}, Radius in Pixels=Constant() {r=1}) {materialFunction=&MF_UI_CoreUI_RoundedMask}.Shape Mask
  )
  n6 = TextureSample(UVs=MaterialFunctionCall(Texture Scale=Constant() {r=3.5}) {materialFunction=&ScaleUVsByCenter}.UVs) {samplerType=SAMPLERTYPE_Grayscale, texture=/Game/Global/Textures/Default/Noise/LowResBlurredNoiseNormalized}
  NamedRerouteDeclaration(None=Multiply(A=LinearInterpolate(Alpha=@n6.R) {constA=-1}, B=Constant() {r=0.25}) {constB=0.2}) {name=Noise, nodeColor=<rgba 0 0 0 1>}
  n7 = ScalarParameter() {defaultValue=0.25, parameterName=Wave_HighestPoint}
  n8 = MaterialFunctionCall() {materialFunction=&GetUserInterfaceUV}
  n9 = ComponentMask(None=Panner(Coordinate=@n8.Normalized UV) {speedX=-0.1}) {r=true}
  n10 = NamedRerouteUsage()
  n11 = Frac(None=Multiply(A=Add(A=@n9, B=@n10), B=ScalarParameter() {defaultValue=16, parameterName=BarsAmount}))
  n12 = SmoothStep(Max=@n7, Value=@n11) {constMax=0.25}
  n13 = SmoothStep(Max=@n7, Value=@n11) {constMax=0.25, constMin=1}
  NamedRerouteDeclaration(None=Multiply(A=Min(A=@n12, B=@n13), B=ComponentMask(None=@n8.Normalized UV) {r=true})) {name=Pattern, nodeColor=<rgba 0 0 0 1>}
  n14 = Constant()
  n15 = ScalarParameter() {defaultValue=4, group=9 - Reveal Mask, parameterName=Reveal Fade}
  n16 = Time()
  n17 = ScalarParameter() {defaultValue=0.3, parameterName=Speed}
  n18 = SmoothStep(Min=ScalarParameter() {defaultValue=0.2, parameterName=Loop Gap}, Value=Frac(None=Multiply(A=@n16, B=@n17) {constB=0.2})) {constMin=0.5}
  n19 = Constant() {r=1}
  n20 = MaterialFunctionCall() {materialFunction=&GetUserInterfaceUV}
  n21 = SmoothStep(
   Min=LinearInterpolate(A=Subtract(A=@n14, B=@n15), Alpha=@n18)
   Max=LinearInterpolate(B=Add(A=@n19, B=@n15), Alpha=@n18)
   Value=Add(A=ComponentMask(None=@n20.Normalized UV) {r=true}, B=NamedRerouteUsage())
  )
  n22 = Constant() {r=0.5}
  n23 = OneMinus(None=Multiply(A=Distance(A=@n21, B=@n22)) {constB=2})
  n24 = MaterialFunctionCall() {materialFunction=&GetUserInterfaceUV}
  n25 = SmoothStep(Min=ScalarParameter() {parameterName=Bloom Amount}, Value=Multiply(A=@n23, B=ComponentMask(None=@n24.Normalized UV) {r=true})) {constMin=0.3}
  NamedRerouteDeclaration(None=Multiply(A=NamedRerouteDeclaration(None=@n25) {name=Edge Glow, nodeColor=<rgba 0 0 0 1>}, B=SmoothStep(Value=@n21) {constMax=0.49, constMin=0.5})) {name=Pattern Mask, nodeColor=<rgba 0 0 0 1>}
  NamedRerouteDeclaration(None=SmoothStep(Min=ScalarParameter() {defaultValue=0.75, parameterName=Edge Size}, Value=@n23) {constMin=0.8}) {name=Edge, nodeColor=<rgba 0 0 0 1>}
 ]
}
