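# CommonButtonBase widget for a Power Ranking button, showing a rank icon, numeric score and "Power Ranking" label.
# A grid layers a dark UIKit backplate, material background, icon, text stack and rounded block outline.
# SetRankIcon applies a soft material instance to the lazy-loading icon, and SetScore writes an integer into the score text.
# Hover and unhover switch the outline between Focused and Unfocused visual states.
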WidgetBlueprint /Compete/PowerRanking/WBP_PowerRanking_Button
graphs [
 EventGraph {nodes [
   Event On Hovered()
   n1 = Get Block_Outline() -> (Block_Outline:WBP UIKit Block Outline Object Reference)
   > RequestTransition!(execute, self:UIKit Block Object Reference=@n1, State:EUIKitBlockVisualState Enum=Focused) -> (then)
   Event On Unhovered()
   > RequestTransition(self=@n1, State=Unfocused)
   Event PreConstruct()
   Event Construct()
   Event Tick()
  ]}
 SetRankIcon {nodes [
   n1 = SetRankIcon!() -> (then, Icon:Material Instance Soft Object Reference) {is=title K2Node_FunctionEntry}
   > SetBrushFromSoftMaterial!(
    execute
    self:Image Object Reference=Get Icon() -> (Icon:Fort Lazy Image Object Reference)
    SoftMaterial:Material Interface Soft Object Reference=@n1
   ) -> (then)
  ]}
 SetScore {nodes [
   n1 = SetScore!() -> (then, Score:int) {is=title K2Node_FunctionEntry}
   > "SetText (Text)"(self=Get Text_Score() -> (Text_Score:Common Text Object Reference), InText="To Text (Integer)"(Value=@n1))
  ]}
]
variables []
widgets {
 info {inheritedWidgetCount=1, parentClass=&CommonButtonBase, rootWidgetClass=&GridPanel, widgetCount=9}
 widgets [
  Root {widgetClassPath=&GridPanel, widget=Root}
  Block_Outline {
   widgetClassPath=/FortUILibrary/UIKit/Buttons/Blocks/WBP_UIKit_Block_Outline
   bIsVariable=true
   parent=Root
   widget=Block_Outline
   properties {bottomLeftCornerRadius=20, bottomRightCornerRadius=20, radiusInPixels=true, topLeftCornerRadius=20, topRightCornerRadius=20}
   slot {class=GridSlot, properties {columnSpan=2, layer=30}}
  }
  WBP_UIKit_Backplate {
   widgetClassPath=&WBP_UIKit_Backplate
   bIsVariable=true
   parent=Root
   widget=WBP_UIKit_Backplate
   properties {brightness=Dark}
   slot {class=GridSlot, properties {columnSpan=2}}
  }
  Background {
   widgetClassPath=&Image
   bIsVariable=true
   parent=Root
   widget=Background
   properties {brush {
     bIsDynamicallyLoaded=false
     drawAs=Image
     imageType=FullColor
     mirroring=NoMirror
     resourceName=None
     resourceObject=/Compete/PowerRanking/Materials/M_PowerRanking_ButtonBackground
     tiling=NoTile
     imageSize=<0 102>
     margin=<ltrb 0 0 0 0>
     outlineSettings {
      bUseBrushTransparency=false
      roundingType=HalfHeightRadius
      width=0
      color {colorUseRule=UseColor_Specified, specifiedColor=<rgba 0 0 0 0>}
      cornerRadii=<0 0 0 0>
     }
     tintColor {colorUseRule=UseColor_Specified, specifiedColor=<rgba 1 1 1 1>}
     uVRegion {bIsValid=false, max=<0 0>, min=<0 0>}
    }}
   slot {class=GridSlot, properties {columnSpan=2, layer=1}}
  }
  Icon {
   widgetClassPath=&FortLazyImage
   bIsVariable=true
   parent=Root
   widget=Icon
   properties {brush {
     bIsDynamicallyLoaded=false
     drawAs=Image
     imageType=FullColor
     mirroring=NoMirror
     resourceName=None
     resourceObject=/Game/UI/Frontend/Lobby/Habanero/Habanero_Materials/MaterialInstances/PowerRanks/MI_UI_PowerRank_Low
     tiling=NoTile
     imageSize=<60 60>
     margin=<ltrb 0 0 0 0>
     outlineSettings {
      bUseBrushTransparency=false
      roundingType=HalfHeightRadius
      width=0
      color {colorUseRule=UseColor_Specified, specifiedColor=<rgba 0 0 0 0>}
      cornerRadii=<0 0 0 0>
     }
     tintColor {colorUseRule=UseColor_Specified, specifiedColor=<rgba 1 1 1 1>}
     uVRegion {bIsValid=false, max=<0 0>, min=<0 0>}
    }}
   slot {class=GridSlot, properties {horizontalAlignment=HAlign_Left, layer=20, verticalAlignment=VAlign_Center, padding=<ltrb 15 0 0 0>}}
  }
  VerticalBox_0 {
   widgetClassPath=&VerticalBox
   parent=Root
   widget=VerticalBox_0
   slot {class=GridSlot, properties {column=1, layer=5, verticalAlignment=VAlign_Center, padding=<ltrb 8 0 24 0>}}
  }
  Text_Score {
   widgetClassPath=&CommonTextBlock
   bIsVariable=true
   parent=VerticalBox_0
   widget=Text_Score
   properties {
    bAutoCollapseWithEmptyText=true
    scrollStyle=&TextScrollStyle-Base_C
    text=0
    font {
     bForceMonospaced=false
     bMaterialIsStencil=false
     fontMaterial=null
     fontObject=&HeadingNow
     letterSpacing=0
     monospacedWidth=1
     size=20.25
     skewAmount=0
     typefaceFontName=75 Medium
     outlineSettings {
      bApplyOutlineToDropShadows=false
      bMiteredCorners=false
      bSeparateFillAlpha=false
      outlineMaterial=null
      outlineSize=0
      outlineColor=<rgba 0 0 0 1>
     }
    }
    shadowOffset=<0 0>
   }
   slot {class=VerticalBoxSlot}
  }
  Text_PowerRanking {
   widgetClassPath=&CommonTextBlock
   parent=VerticalBox_0
   widget=Text_PowerRanking
   properties {
    scrollStyle=&TextScrollStyle-Base_C
    text=Power Ranking
    font {
     bForceMonospaced=false
     bMaterialIsStencil=false
     fontMaterial=null
     fontObject=&HeadingNow
     letterSpacing=0
     monospacedWidth=1
     size=14.25
     skewAmount=0
     typefaceFontName=74 Regular
     outlineSettings {
      bApplyOutlineToDropShadows=false
      bMiteredCorners=false
      bSeparateFillAlpha=false
      outlineMaterial=null
      outlineSize=0
      outlineColor=<rgba 0 0 0 1>
     }
    }
    shadowOffset=<0 0>
   }
   slot {class=VerticalBoxSlot}
  }
  InputActionWidget {
   widgetClassPath=&CommonActionWidget
   bInherited=true
   bIsVariable=true
   parent=Root
   widget=InputActionWidget
   properties {inputActions [{dataTable=null, rowName=None}]}
   slot {class=GridSlot, properties {horizontalAlignment=HAlign_Left, verticalAlignment=VAlign_Center, nudge=<-36 0>}}
  }
 ]
}
