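# M_TextScoreGradient is a translucent UI material that makes a two-color gradient, apparently for score text in the Power Ranking UI, judging by its name and folder.
# It linearly interpolates between the Color1 and Color2 vector parameters, defaulting to blue and light purple.
# The red channel of the texture coordinates drives the blend, so the gradient runs horizontally along U.
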Material /Compete/PowerRanking/Materials/M_TextScoreGradient
graph {
 properties {bAutomaticallySetUsageInEditor=false, blendMode=BLEND_Translucent, materialDomain=MD_UI, usedShadingModels=Default Lit}
 nodes [LinearInterpolate(
   A=VectorParameter() {parameterName=Color1, defaultValue=<rgba 0.0529 0.314 0.8632 1>}.RGB
   B=VectorParameter() {parameterName=Color2, defaultValue=<rgba 0.5972 0.2462 1 1>}.RGB
   Alpha=ComponentMask(None=TextureCoordinate()) {r=true}
  )]
}
