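# Translucent UI material for the Power Ranking flash: a purple star flare with glow and drifting particles.
# A rotated cross falloff masked by a radial gradient forms the star, combined with a horizontal line, a soft glow and a ping-pong pulse.
# Particles come from a thresholded, vertically scrolling star texture inside a radial mask.
# The combined mask tints a reflected-gradient background and merges with that gradient into a second, alpha-style mask.
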Material /Compete/PowerRanking/Materials/M_UI_PowerRanking_Flash
graph {
 properties {bAutomaticallySetUsageInEditor=false, blendMode=BLEND_Translucent, materialDomain=MD_UI, usedShadingModels=Default Lit}
 nodes [
  n1 = Power(Base=MaterialFunctionCall!(UVs, Offset, RotationCenter, RotationAngle=Constant() {r=0.25}) {materialFunction=/FortUILibrary/MaterialFunctions/Gradients/MF_UI_ReflectedGradient}.Result)
  n2 = MaterialFunctionCall(UVs=NamedRerouteUsage(), "Rotation Angle (0-1)"=ScalarParameter() {defaultValue=0.125, parameterName=StarRotation}) {materialFunction=&CustomRotator}
  n3 = Multiply(A=Abs(None=Subtract(A=@n2.Rotated Values) {constB=0.5})) {constB=2}
  n4 = OneMinus(None=Multiply(A=ComponentMask(None=@n3) {r=true}, B=ComponentMask(None=@n3) {g=true}))
  n5 = ScalarParameter() {defaultValue=60, parameterName=Flare Strength}
  n6 = Divide(A=Saturate(None=Power(Base=@n4, Exp=@n5) {constExponent=120}))
  n7 = MaterialFunctionCall(UVs=NamedRerouteUsage(), Radius=ScalarParameter() {defaultValue=0.4, parameterName=StarRadius}, Density=Constant() {r=1}) {materialFunction=&RadialGradientExponential}
  n8 = NamedRerouteUsage()
  n9 = Multiply(A=Abs(None=Subtract(A=@n8) {constB=0.5})) {constB=2}
  n10 = Power(Base=OneMinus(None=ComponentMask(None=@n9) {g=true}), Exp=ScalarParameter() {defaultValue=120, parameterName=LineDedsity}) {constExponent=120}
  n11 = ScalarParameter() {defaultValue=1, parameterName=LineOpacity}
  n12 = Multiply(A=Max(A=Multiply(A=@n6, B=@n7.RadialGradientExponential), B=Multiply(A=@n10, B=@n11)), B=NamedRerouteUsage())
  n13 = MaterialFunctionCall(UVs=NamedRerouteUsage(), Radius=ScalarParameter() {defaultValue=0.35, parameterName=GlowRadius}, Density=Constant() {r=1}) {materialFunction=&RadialGradientExponential}
  n14 = Saturate(None=Add(A=Add(A=@n12, B=@n13.RadialGradientExponential), B=NamedRerouteUsage()))
  LinearInterpolate(
   A=LinearInterpolate(
    A=Constant4Vector() {constant=<rgba 0.0009 0.0021 0.0033 1>}.RGBA
    B=Constant4Vector() {constant=<rgba 0.0273 0.003 0.1144 1>}.RGBA
    Alpha=Saturate(None=@n1)
   )
   B=Constant4Vector() {constant=<rgba 0.3824 0.0491 0.9615 1>}.RGBA
   Alpha=@n14
  )
  n15 = Constant() {r=0.2}
  n16 = Constant() {r=0.7}
  n17 = Multiply(A=Time()) {constB=0.02}
  n18 = Add(
   A=MaterialFunctionCall(UVs=NamedRerouteUsage(), Texture Scale=AppendVector(A=@n16, B=@n16)) {materialFunction=&ScaleUVsByCenter}.UVs
   B=AppendVector(A=Constant(), B=Frac(None=@n17))
  )
  n19 = TextureObject() {samplerType=SAMPLERTYPE_Alpha, texture=/Game/UI/Frontend/Lobby/Habanero/Habanero_Textures/T_UI_Habanero_Clouds_Stars}
  n20 = Multiply(A=SmoothStep(Min=OneMinus(None=@n15), Value=TextureSample(UVs=@n18, Tex=@n19).R) {constMin=0.5}) {constB=5}
  n21 = MaterialFunctionCall(UVs=NamedRerouteUsage(), Texture Scale=Constant2Vector() {g=1.3, r=1}) {materialFunction=&ScaleUVsByCenter}
  n22 = Constant() {r=0.4}
  n23 = Constant() {r=1}
  NamedRerouteDeclaration(None=Multiply(
    A=Saturate(None=@n20)
    B=MaterialFunctionCall(UVs=@n21.UVs, Radius=@n22, Density=@n23) {materialFunction=&RadialGradientExponential}.RadialGradientExponential
   )) {name=Particles, nodeColor=<rgba 1 0.6563 0 1>}
  NamedRerouteDeclaration(None=MaterialFunctionCall() {materialFunction=&MF_UI_UVScaling}.Scaled to Fit) {name=UV, nodeColor=<rgba 0 1 0.1328 1>}
  n24 = MaterialFunctionCall!(FloatValue, Speed) {materialFunction=/FortUILibrary/MaterialFunctions/Basics/MF_UI_LinearTime}
  NamedRerouteDeclaration(None=Add(A=Multiply(A=@n24.0-1 Ping pong) {constB=0.5}) {constB=0.5}) {name=Time, nodeColor=<rgba 1 0 0.3984 1>}
  Saturate(None=Add(A=Max(A=@n14, B=@n1)))
 ]
}
