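# Full-screen activatable overlay announcing a Power Ranking upgrade, with a rank emblem, score gain and season caption.
# On activation it plays Intro, which drives star, glow and line parameters on the flash background material, pops in the emblem, then staggers in the text.
# The Close button starts collapsed, appears late in Intro and deactivates the widget when clicked.
# Rank image and texts are fixed designer values; this graph never sets them.
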WidgetBlueprint /Compete/PowerRanking/WBP_PowerRanking_Overlay
animations [Intro {
  binding_count=6
  length_seconds=2.2
  object=&self:Intro
  sequence {bindings [
    WBP_UIKit_Button_Regular {
     id=#1
     possessed_class=WBP_UIKit_Button_Regular_C
     tracks [
      {
       type=MovieSceneFloatTrack
       display_name=Render Opacity
       sections [{
         type=MovieSceneFloatSection
         end_frame=44
         start_frame=0
         channels [None {type=MovieSceneScriptingFloatChannel, thinned_from=4, keys [[0,0],[40,0],[44,1]]}]
        }]
      }
      {
       type=MovieSceneByteTrack
       display_name=Visibility
       sections [{type=MovieSceneByteSection, end_frame=39, start_frame=0, channels [None {type=MovieSceneScriptingByteChannel, keys [[0,1],[4,1],[39,0]]}]}]
      }
     ]
    }
    Rank {
     id=#2
     possessed_class=Image
     tracks [
      {
       type=MovieSceneFloatTrack
       display_name=Render Opacity
       sections [{
         type=MovieSceneFloatSection
         end_frame=11
         start_frame=0
         channels [None {type=MovieSceneScriptingFloatChannel, thinned_from=4, keys [[0,0],[7,0],[11,1]]}]
        }]
      }
      {
       type=MovieScene2DTransformTrack
       sections [{
         type=MovieScene2DTransformSection
         end_frame=20
         start_frame=0
         channels [
          Translation.X {type=MovieSceneScriptingFloatChannel, const [0,2]}
          Translation.Y {type=MovieSceneScriptingFloatChannel, const [0,2]}
          Angle {type=MovieSceneScriptingFloatChannel, const [0,2]}
          Scale.X {type=MovieSceneScriptingFloatChannel, thinned_from=5, keys [[0,0.5],[7,0.5],[12,1.3],[20,1]]}
          Scale.Y {type=MovieSceneScriptingFloatChannel, thinned_from=5, keys [[0,0.5],[7,0.5],[12,1.3],[20,1]]}
          Shear.X {type=MovieSceneScriptingFloatChannel, const [0,2]}
          Shear.Y {type=MovieSceneScriptingFloatChannel, const [0,2]}
         ]
        }]
      }
     ]
    }
    Score {
     id=#3
     possessed_class=CommonTextBlock
     tracks [{
       type=MovieSceneFloatTrack
       display_name=Render Opacity
       sections [{
         type=MovieSceneFloatSection
         end_frame=33
         start_frame=0
         channels [None {type=MovieSceneScriptingFloatChannel, thinned_from=4, keys [[0,0],[27,0],[33,1]]}]
        }]
      }]
    }
    Foter {
     id=#4
     possessed_class=CommonTextBlock
     tracks [{
       type=MovieSceneFloatTrack
       display_name=Render Opacity
       sections [{
         type=MovieSceneFloatSection
         end_frame=36
         start_frame=0
         channels [None {type=MovieSceneScriptingFloatChannel, thinned_from=4, keys [[0,0],[30,0],[36,1]]}]
        }]
      }]
    }
    Header {
     id=#5
     possessed_class=CommonTextBlock
     tracks [{
       type=MovieSceneFloatTrack
       display_name=Render Opacity
       sections [{
         type=MovieSceneFloatSection
         end_frame=30
         start_frame=0
         channels [None {type=MovieSceneScriptingFloatChannel, thinned_from=4, keys [[0,0],[24,0],[30,1]]}]
        }]
      }]
    }
    Background {
     id=#6
     possessed_class=Image
     tracks [{
       type=MovieSceneWidgetMaterialTrack
       sections [{
         type=MovieSceneParameterSection
         end_frame=42
         start_frame=0
         channels [
          StarRadius {type=MovieSceneScriptingFloatChannel, thinned_from=6, keys [[0,0],[6,0],[9,0.8],[12,0.8],[20,0.5]]}
          GlowRadius {type=MovieSceneScriptingFloatChannel, keys [[0,0],[4,0],[12,0.5],[20,0.4]]}
          StarRotation {type=MovieSceneScriptingFloatChannel, thinned_from=5, keys [[0,0.125],[12,0.125],[20,-0.125]]}
          LineDedsity {type=MovieSceneScriptingFloatChannel, keys [[0,120],[4,120],[5.98,20],[20,120]]}
          LineOpacity {type=MovieSceneScriptingFloatChannel, keys [[0,0],[4,0],[5.98,1]]}
         ]
        }]
      }]
    }
   ]}
 }]
graphs [EventGraph {nodes [
   "On Clicked (WBP_UIKit_Button_Regular)"!() -> (OutputDelegate, then, Button:Common Button Base Object Reference) {is=title K2Node_ComponentBoundEvent}
   > DeactivateWidget()
   Event On Activated()
   > PlayAnimation(InAnimation=Get Intro())
  ]}]
variables []
widgets {
 info {inheritedWidgetCount=0, parentClass=&CommonActivatableWidget, rootWidgetClass=&GridPanel, widgetCount=11}
 widgets [
  Root {widgetClassPath=&GridPanel, widget=Root, properties {columnFill [1], rowFill [1]}}
  Background {
   widgetClassPath=&Image
   bIsVariable=true
   parent=Root
   widget=Background
   properties {brush {
     bIsDynamicallyLoaded=false
     drawAs=Image
     imageType=FullColor
     mirroring=NoMirror
     resourceName=None
     resourceObject=/Compete/PowerRanking/Materials/M_UI_PowerRanking_Flash
     tiling=NoTile
     imageSize=<32 32>
     margin=<ltrb 0 0 0 0>
     outlineSettings {
      bUseBrushTransparency=false
      roundingType=HalfHeightRadius
      width=0
      color {colorUseRule=UseColor_Specified, specifiedColor=<rgba 0 0 0 0>}
      cornerRadii=<0 0 0 0>
     }
     tintColor {colorUseRule=UseColor_Specified, specifiedColor=<rgba 1 1 1 1>}
     uVRegion {bIsValid=false, max=<0 0>, min=<0 0>}
    }}
   slot {class=GridSlot}
  }
  VerticalBox_0 {
   widgetClassPath=&VerticalBox
   parent=Root
   widget=VerticalBox_0
   slot {class=GridSlot, properties {horizontalAlignment=HAlign_Center, verticalAlignment=VAlign_Bottom, padding=<ltrb 0 0 0 80>}}
  }
  Header {
   widgetClassPath=&CommonTextBlock
   parent=VerticalBox_0
   widget=Header
   properties {
    justification=Center
    text="Power Ranking\r\nUpgraded!"
    textTransformPolicy=ToUpper
    font {
     bForceMonospaced=false
     bMaterialIsStencil=false
     fontMaterial=null
     fontObject=&HeadingNow
     letterSpacing=0
     monospacedWidth=1
     size=36
     skewAmount=0
     typefaceFontName=86 Bold Italic
     outlineSettings {
      bApplyOutlineToDropShadows=false
      bMiteredCorners=false
      bSeparateFillAlpha=false
      outlineMaterial=null
      outlineSize=0
      outlineColor=<rgba 0 0 0 1>
     }
    }
    shadowOffset=<0 0>
   }
   slot {class=VerticalBoxSlot, properties {horizontalAlignment=HAlign_Center}}
  }
  Score {
   widgetClassPath=&CommonTextBlock
   bIsVariable=true
   parent=VerticalBox_0
   widget=Score
   properties {
    text="+2,313"
    font {
     bForceMonospaced=false
     bMaterialIsStencil=false
     fontMaterial=/Compete/PowerRanking/Materials/M_TextScoreGradient
     fontObject=&HeadingNow
     letterSpacing=0
     monospacedWidth=1
     size=36
     skewAmount=0
     typefaceFontName=86 Bold Italic
     outlineSettings {
      bApplyOutlineToDropShadows=false
      bMiteredCorners=false
      bSeparateFillAlpha=false
      outlineMaterial=null
      outlineSize=0
      outlineColor=<rgba 0 0 0 1>
     }
    }
    shadowOffset=<0 0>
   }
   slot {class=VerticalBoxSlot, properties {horizontalAlignment=HAlign_Center}}
  }
  Foter {
   widgetClassPath=&CommonTextBlock
   parent=VerticalBox_0
   widget=Foter
   properties {
    text=From Last Season
    textTransformPolicy=ToUpper
    colorAndOpacity {colorUseRule=UseColor_Specified, specifiedColor=<rgba 0.5776 0.5776 0.5776 1>}
    font {
     bForceMonospaced=false
     bMaterialIsStencil=false
     fontMaterial=null
     fontObject=&HeadingNow
     letterSpacing=0
     monospacedWidth=1
     size=20.25
     skewAmount=0
     typefaceFontName=75 Medium
     outlineSettings {
      bApplyOutlineToDropShadows=false
      bMiteredCorners=false
      bSeparateFillAlpha=false
      outlineMaterial=null
      outlineSize=0
      outlineColor=<rgba 0 0 0 1>
     }
    }
    shadowOffset=<0 0>
   }
   slot {class=VerticalBoxSlot, properties {horizontalAlignment=HAlign_Center}}
  }
  Rank {
   widgetClassPath=&Image
   bIsVariable=true
   parent=Root
   widget=Rank
   properties {brush {
     bIsDynamicallyLoaded=false
     drawAs=Image
     imageType=FullColor
     mirroring=NoMirror
     resourceName=None
     resourceObject=/Game/UI/Frontend/Lobby/Habanero/Habanero_Materials/MaterialInstances/PowerRanks/MI_UI_PowerRank_Hyper
     tiling=NoTile
     imageSize=<500 500>
     margin=<ltrb 0 0 0 0>
     outlineSettings {
      bUseBrushTransparency=false
      roundingType=HalfHeightRadius
      width=0
      color {colorUseRule=UseColor_Specified, specifiedColor=<rgba 0 0 0 0>}
      cornerRadii=<0 0 0 0>
     }
     tintColor {colorUseRule=UseColor_Specified, specifiedColor=<rgba 1 1 1 1>}
     uVRegion {bIsValid=false, max=<0 0>, min=<0 0>}
    }}
   slot {class=GridSlot, properties {horizontalAlignment=HAlign_Center, verticalAlignment=VAlign_Center}}
  }
  SafeZone_81 {
   widgetClassPath=&SafeZone
   bIsVariable=true
   parent=Root
   widget=SafeZone_81
   slot {class=GridSlot, properties {horizontalAlignment=HAlign_Right, verticalAlignment=VAlign_Bottom}}
  }
  Overlay_0 {
   widgetClassPath=&Overlay
   parent=SafeZone_81
   widget=Overlay_0
   slot {class=SafeZoneSlot, properties {hAlign=HAlign_Left, vAlign=VAlign_Top, padding=<ltrb 0 0 40 20>}}
  }
  WBP_CaptureForPostBufferUpdate {
   widgetClassPath=&WBP_CaptureForPostBufferUpdate
   bIsVariable=true
   parent=Overlay_0
   slot {class=OverlaySlot, properties {horizontalAlignment=HAlign_Fill, verticalAlignment=VAlign_Fill}}
  }
  WBP_UIKit_Button_Regular {
   widgetClassPath=&WBP_UIKit_Button_Regular
   bIsVariable=true
   parent=Overlay_0
   widget=WBP_UIKit_Button_Regular
   properties {
    navigation=&self:WidgetTree.WBP_UIKit_Button_Regular.WidgetNavigation_0
    text=Close
    visibility=Collapsed
    triggeringInputAction {dataTable=&InputActionDataTable, rowName=Close}
   }
   slot {class=OverlaySlot}
  }
 ]
}
